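# FaceAcc is a Default Lit material in a DEMO mesh folder of the MollyKit tail accessory, likely an imported placeholder for a face-accessory part.
# Its graph holds a single pink Constant3Vector color, with no textures, parameters or visible connection to a material output.
# Interchange asset import data indicates it was generated by a mesh or scene import.
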Material /MiscReferencersOfCosmetics/Accessories/FORT_Tails/Tail_F_MED_MollyKit/Meshes/DEMO/FaceAcc
graph {
 properties {assetImportData="<self>:InterchangeAssetImportData_0", usedShadingModels=Default Lit}
 nodes [Constant3Vector() {constant=<rgba 0.7682 0.2118 0.3572 1>}]
}
